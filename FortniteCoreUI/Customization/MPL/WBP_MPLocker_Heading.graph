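# Locker heading widget that shows an uppercase header and a smaller, optional subheader stacked in a grid.
# SetHeader and SetSubHeader set the text, while SetSubHeadingVisibility toggles the subheader between collapsed and hit-test invisible.
# PreConstruct fills placeholder text at design time, then applies Show Sub Heading? and UseNewLockerTextSize.
# When UseNewLockerTextSize is true, the header's StyleSheet gameplay tag swaps from Style.40 to Style.20.
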WidgetBlueprint /FortniteCoreUI/Customization/MPL/WBP_MPLocker_Heading
graphs [
 EventGraph {nodes [
   n1 = Event PreConstruct()
   > n2 = Sequence!(execute) -> (then_0, then_1, then_2:Exec)
   > Branch(Condition=@n1.IsDesignTime)
   > PresetForDesignTime()
   SetSubHeadingVisibility!(execute=@n2.then_1, self, ShowSubHeading?:bool=Get Show Sub Heading?() -> (Show Sub Heading?:bool)) -> (then)
   UpdateHeadingStyling!(execute=@n2.then_2, self, UseNewStyling:bool=Get UseNewLockerTextSize() -> (UseNewLockerTextSize:bool)) -> (then)
   Event Construct()
   Event Tick()
  ]}
 PresetForDesignTime {nodes [
   PresetForDesignTime() {is=title K2Node_FunctionEntry}
   > Set Text(Text="LOCTEXT(\"Heading\")", self=Get Text_Header() -> (Text_Header:Common Text Object Reference))
   > Set Text(Text="LOCTEXT(\"Sub Heading\")", self=Get Text_SubHeader() -> (Text_SubHeader:Common Text Object Reference))
   > SetSubHeadingVisibility(ShowSubHeading?=true)
  ]}
 SetHeader {nodes [
   n1 = SetHeader!() -> (then, Text:Text) {is=title K2Node_FunctionEntry}
   > Set Text(Text=@n1, self=Get Text_Header())
  ]}
 SetSubHeader {nodes [
   n1 = SetSubHeader!() -> (then, Text:Text) {is=title K2Node_FunctionEntry}
   > Set Text(Text=@n1, self=Get Text_SubHeader())
  ]}
 SetSubHeadingVisibility {nodes [
   n1 = SetSubHeadingVisibility!() -> (then, ShowSubHeading?:bool=~) {is=title K2Node_FunctionEntry}
   > SetVisibility(self=Get Text_SubHeader(), InVisibility=Select(Option 0=Collapsed, Option 1=HitTestInvisible, Index=@n1))
  ]}
 UpdateHeadingStyling {nodes [
   n1 = UpdateHeadingStyling!() -> (then, UseNewStyling:bool=~) {is=title K2Node_FunctionEntry}
   > Branch(Condition=@n1)
   > UpdateWidgetStyleTags!(
    execute
    InWidget:Widget Object Reference=Get Text_Header()
    InAddedTags:Array of Gameplay Tag Structures=Make Array("[0]"=(TagName="StyleSheet.Style.20"))
    InRemovedTags:Array of Gameplay Tag Structures=Make Array("[0]"=(TagName="StyleSheet.Style.40"))
   ) -> (then, ReturnValue:bool)
  ]}
]
variables [
 Show Sub Heading? {type=boolean, default=true}
 UseNewLockerTextSize {type=boolean}
]
widgets {
 info {inheritedWidgetCount=0, parentClass=&UserWidget, rootWidgetClass=&GridPanel, widgetCount=3}
 widgets [
  GridPanel_16 {widgetClassPath=&GridPanel, widget=GridPanel_16, properties {columnFill [1], rowFill [1]}}
  Text_Header {
   widgetClassPath=&CommonTextBlock
   bIsVariable=true
   parent=GridPanel_16
   widget=Text_Header
   properties {
    mobileFontSizeMultiplier=0.75
    scrollStyle=/FortUILibrary/Styles/Typography/Foundation/TextScrollStyle-NoFade.TextScrollStyle-NoFade_C
    text=HEADING
    textTransformPolicy=ToUpper
    font {
     bForceMonospaced=false
     bMaterialIsStencil=false
     fontMaterial=null
     fontObject=&HeadingNow
     letterSpacing=-40
     monospacedWidth=1
     size=51.75
     skewAmount=0
     typefaceFontName=86 Bold Italic
     outlineSettings {
      bApplyOutlineToDropShadows=false
      bMiteredCorners=false
      bSeparateFillAlpha=false
      outlineMaterial=null
      outlineSize=0
      outlineColor=<rgba 0 0 0 1>
     }
    }
    shadowOffset=<0 0>
   }
   slot {class=GridSlot}
  }
  Text_SubHeader {
   widgetClassPath=&CommonTextBlock
   bIsVariable=true
   parent=GridPanel_16
   widget=Text_SubHeader
   properties {
    mobileFontSizeMultiplier=0.75
    scrollStyle=&TextScrollStyle-NoFade_C
    text=LOADOUT
    textTransformPolicy=ToUpper
    font {
     bForceMonospaced=false
     bMaterialIsStencil=false
     fontMaterial=/FortUILibrary/Materials/Text/M_UI_Additive_Font
     fontObject=&HeadingNow
     letterSpacing=-10
     monospacedWidth=1
     size=25.5
     skewAmount=0
     typefaceFontName=86 Bold Italic
     outlineSettings {
      bApplyOutlineToDropShadows=false
      bMiteredCorners=false
      bSeparateFillAlpha=false
      outlineMaterial=null
      outlineSize=0
      outlineColor=<rgba 0 0 0 1>
     }
    }
    shadowOffset=<0 0>
   }
   slot {class=GridSlot, properties {rowSpan=2, verticalAlignment=VAlign_Bottom, padding=<ltrb 0 40 0 0>}}
  }
 ]
}
