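# MPL variant of the UIKit item description widget, adding an access-restriction warning box to the base layout for cosmetic item details.
# Restriction changes toggle the box and choose between a gating warning icon, a Juno-style "detailed figure incoming" icon, or none.
# View models drive category text with fallbacks and let nameable Mimosa cosmetics show their item name as title.
# It also remaps rarity palette colors by series and caps title height.
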WidgetBlueprint /FortniteCoreUI/Customization/MPL/WBP_UIKit_ItemDescription_MPL
graphs [
 Access Restrictions Changed {nodes [
   n1 = Access Restrictions Changed!() -> (then, Restriction:EItemAccessRestriction Enum=UnsatisfiedExclusiveItem) {is=title K2Node_FunctionEntry}
   > Set Visibility(
    Visibility="To Visibility (Boolean)"!(
     bIsVisible:bool="Equal (Enum)"(A:EItemAccessRestriction Enum=@n1, B:EItemAccessRestriction Enum=Unknown)
     TrueVisibility:ESlateVisibility Enum=Collapsed
     FalseVisibility:ESlateVisibility Enum=HitTestInvisible
    ) -> (ReturnValue:ESlateVisibility Enum=Visible)
    self=Get AccessRestriction_Warning() -> (AccessRestriction_Warning:WBP UIKit Info Box Object Reference)
   )
   n2 = Get AccessRestriction_Warning()
   > SetShowIcon!(execute, self:WBP UIKit Info Box Object Reference=@n2, Show Icon:bool=false) -> (then)
   > SetIconSize!(execute, self:WBP UIKit Info Box Object Reference=@n2, InSize:Vector 2D Structure=(X=40,Y=40)) -> (then)
   > SetIconImage!(
    execute
    self:WBP UIKit Info Box Object Reference=@n2
    Material:Material Interface Object Reference=/Game/Athena/UI/Frontend/CosmeticGating/Materials/MI_UI_CosmeticGate_WarningIcon
   ) -> (then)
   > n3 = Branch(Condition=Select!(
     UnsatisfiedExclusiveItem:bool=false
     LockedEmote:bool=false
     CosmeticRestrictions:bool=false
     AllArchived:bool=false
     RatingGated:bool=true
     UnsupportedProduct:bool=false
     AdditionalInfo:bool=false
     AdditionalInfoWarning:bool=true
     Unknown:bool=false
     Index:EItemAccessRestriction Enum=Get Restriction() -> (Restriction:EItemAccessRestriction Enum=UnsatisfiedExclusiveItem)
    ) -> (ReturnValue:bool))
   n4 = Get AccessRestriction_Warning()
   > SetShowIcon(self=@n4, Show Icon=true)
   > SetIconSize(self=@n4, InSize=(X=40,Y=40))
   > SetIconImage(self=@n4, Material=&MI_UI_CosmeticGate_WarningIcon)
   isDetailedFigureIncoming!(
    execute=@n3.else
    self
    isDetailedFigureIncoming:bool=Get bMoreDetailedFigureIncoming!(self:Fort Item Details VM Object Reference=Get FortItemDetailsVM() -> (FortItemDetailsVM:Fort Item Details VM Object Reference)) -> (bMoreDetailedFigureIncoming:bool)
   ) -> (then)
  ]}
 EventGraph {nodes [
   n1 = Event PreConstruct()
   > Parent: PreConstruct(IsDesignTime=@n1.IsDesignTime)
   > SetMaxDesiredHeight!(execute, self:Size Box Object Reference=Get SizeBoxTitle() -> (SizeBoxTitle:Size Box Object Reference), InMaxDesiredHeight:float=210) -> (then)
  ]}
 Get Category Text {nodes [
   n1 = Get Category Text!() -> (then, Item Text:Text, Category Text:Text) {is=title K2Node_FunctionEntry}
   > Return Node!(execute, Result:Text=SelectText(A=@n1, B=@n1.Category Text, bPickA=TextIsEmpty(InText=@n1.Category Text)))
  ]}
 GetDisplayTitle {nodes [
   GetDisplayTitle!() -> (then, DisplayNameText:Text, NickName:Text) {is=title K2Node_FunctionEntry}
   > n1 = Branch(Condition=Is Mimosa!() -> (Bool:bool=~) {is=title K2Node_MacroInstance})
   n2 = GetMimosaName!(self) -> (NameText:Text, IsNameEmpty:bool, IsNamable:bool)
   > n3 = Branch(Condition=OR Boolean(A=NOT Boolean(A=@n2.IsNameEmpty), B=@n2.IsNamable))
   > Return Node!(execute, Title:Text=@n2)
   Return Node(execute=[@n3.else, @n1.else], Title=Get DisplayNameText() -> (DisplayNameText:Text))
   Get NickName() -> (NickName:Text)
  ]}
 GetMimosaName {nodes [
   GetMimosaName() {is=title K2Node_FunctionEntry}
   > n1 = IsNameable!(execute:Exec, self:Object Reference=Get FortItemDetailsVM()) -> (then:Exec, ReturnValue:bool) {is=title K2Node_Message}
   > n2 = Branch(Condition=@n1)
   > n3 = GetItemName!(execute:Exec, self:Object Reference=Get FortItemDetailsVM()) -> (then:Exec, ReturnValue:Text) {is=title K2Node_Message}
   > Return Node!(execute, NameText:Text=@n3, IsNameEmpty:bool=TextIsEmpty(InText=@n3), IsNamable:bool=true)
   Return Node(execute=@n2.else, IsNameEmpty=true, IsNamable=false)
  ]}
 GetRarityColors {nodes [
   n1 = GetRarityColors!() -> (then, IsSeries:bool=~, FortColorPalette:"Fort Color Palette Structure (by ref)") {is=title K2Node_FunctionEntry}
   n2 = Break Fort Color Palette!(FortColorPalette:"Fort Color Palette Structure (by ref)"=@n1.FortColorPalette) -> (
    Color1:Linear Color Structure
    Color2:Linear Color Structure
    Color3:Linear Color Structure
    Color4:Linear Color Structure
    Color5:Linear Color Structure
   ) {is=title K2Node_BreakStruct}
   > Return Node!(
    execute
    RarityColors:Fort Color Palette Structure=Make Fort Color Palette!(
     Color1:Linear Color Structure=SelectColor(A=@n2, B=@n2.Color2, bPickA=@n1)
     Color2:Linear Color Structure=@n2.Color4
     Color3:Linear Color Structure=(R=0,G=0,B=0,A=1)
     Color4:Linear Color Structure=(R=0,G=0,B=0,A=1)
     Color5:Linear Color Structure=(R=0,G=0,B=0,A=1)
    ) -> (FortColorPalette:Fort Color Palette Structure) {is=title K2Node_MakeStruct}
   )
  ]}
 IsMimosa {nodes [
   Inputs!()
   n1 = Get ItemDefinition!(self:Fort Item VM Object Reference=Get Item!(self:Fort Item Details VM Object Reference=Get FortItemDetailsVM()) -> (Item:Fort Item VM Object Reference)) -> (ItemDefinition:Fort Item Definition Object Reference)
   Outputs!(Bool:bool="Equal (Enum)"(
     A:EFortItemType Enum=GetItemType!(ItemDefinition:Item Definition Base Object Reference=@n1) -> (ReturnValue:EFortItemType Enum=WorldItem)
     B:EFortItemType Enum=CosmeticMimosa
    ))
  ]}
 Update Category Text From Category Item {nodes [
   n1 = Update Category Text From Category Item!() -> (then, Category Text:Text) {is=title K2Node_FunctionEntry}
   n2 = Get FortItemDetailsVM()
   > n3 = Is Valid(InputObject=@n2)
   SetCategoryText!(execute=@n3.Is Not Valid, self:WBP UIKit Item Description Base Object Reference, Text:Text=@n1) -> (then)
   SetCategoryText(
    execute=@n3.Is Valid
    Text=Get Category Text!(self, Item Text:Text=GetItemCategoryName!(self:Fort Item Details VM Object Reference=@n2) -> (ReturnValue:Text), Category Text:Text=@n1) -> (Result:Text) {is=title K2Node_CallFunction}
   )
  ]}
 Update Category Text From Item {nodes [
   n1 = Update Category Text From Item!() -> (then, Category Text:Text) {is=title K2Node_FunctionEntry}
   n2 = Get FortLockerCategoryItemVM() -> (FortLockerCategoryItemVM:Fort Locker Category Item VM Object Reference)
   > n3 = Is Valid(InputObject=@n2)
   SetCategoryText(execute=@n3.Is Not Valid, Text=@n1)
   SetCategoryText(
    execute=@n3.Is Valid
    Text=Get Category Text(Item Text=@n1, Category Text=GetCategoryDisplayName!(self:Fort Locker Category Item VM Object Reference=@n2) -> (ReturnValue:Text))
   )
  ]}
 isDetailedFigureIncoming {nodes [
   n1 = isDetailedFigureIncoming!() -> (then, isDetailedFigureIncoming:bool=~) {is=title K2Node_FunctionEntry}
   > n2 = Branch(Condition=Get bHasAccessRestriction!(self:Fort Item Details VM Object Reference=Get FortItemDetailsVM()) -> (bHasAccessRestriction:bool))
   Branch(execute=@n2.else, Condition=@n1)
   n3 = Get AccessRestriction_Warning()
   > SetShowIcon(self=@n3, Show Icon=true)
   > SetIconSize(self=@n3, InSize=(X=40,Y=40))
   > SetIconImage(self=@n3, Material=/Game/Athena/UI/Frontend/CosmeticGating/Materials/MI_ItemJunoStyle)
  ]}
]
variables []
widgets {
 info {
  inheritedWidgetCount=0
  namedSlotCount=1
  parentClass=/FortUILibrary/Systems/Items/WBP_UIKit_ItemDescription_Base.WBP_UIKit_ItemDescription_Base_C
  rootWidgetClass=null
  widgetCount=1
 }
 widgets [AccessRestriction_Warning {
   widgetClassPath=/FortUILibrary/UIKit/InfoBox/WBP_UIKit_InfoBox
   bIsVariable=true
   widget=AccessRestriction_Warning
   properties {showIcon=false, visibility=Collapsed}
  }]
}
